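# Stereo one-shot MetaSound for NitroDrome firework sparks: a random burst, looping sparks and a random ending.
# Play starts burst, loop and a timer of about three seconds; NewEvent either restarts the timer or retriggers, depending on loop state.
# A second loop layer is gated on that state after a delay.
# Timer expiry clears looping, plays the ending and releases the loops' envelope; a trigger accumulator gates OnFinished.
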MetaSoundSource /Game/Sounds/Helios/POI/NitroDrome/MSS_Helios_Nitrodrome_Sparks
inputs [
 LoopingOnshotSource {type=WaveAsset:Array}
 NewEvent {type=Trigger}
 OnFinished {type=WaveAsset:Array}
 Source {type=WaveAsset:Array}
 UE.Source.OnPlay {type=Trigger}
]
interfaces [UE.OutputFormat.Stereo:1.0, UE.Source.OneShot:1.0, UE.Source:1.1]
outputs [
 "UE.OutputFormat.Stereo.Audio:0" {type=Audio}
 "UE.OutputFormat.Stereo.Audio:1" {type=Audio}
 UE.Source.OneShot.OnFinished {type=Trigger}
]
pages [Default {
  input_defaults {
   NewEvent=null
   UE.Source.OnPlay=null
   LoopingOnshotSource [
    /Game/Sounds/Helios/POI/NitroDrome/Fireworks/NitroDrome_Fireworks_Sparks_01
    /Game/Sounds/Helios/POI/NitroDrome/Fireworks/NitroDrome_Fireworks_Sparks_02
    /Game/Sounds/Helios/POI/NitroDrome/Fireworks/NitroDrome_Fireworks_Sparks_03
    /Game/Sounds/Helios/POI/NitroDrome/Fireworks/NitroDrome_Fireworks_Sparks_04
   ]
   OnFinished [
    /Game/Sounds/Helios/POI/NitroDrome/Fireworks/NitroDrome_Fireworks_Sparks_End_04
    /Game/Sounds/Helios/POI/NitroDrome/Fireworks/NitroDrome_Fireworks_Sparks_End_03
    /Game/Sounds/Helios/POI/NitroDrome/Fireworks/NitroDrome_Fireworks_Sparks_End_02
    /Game/Sounds/Helios/POI/NitroDrome/Fireworks/NitroDrome_Fireworks_Sparks_End_01
   ]
   Source [
    /Game/Sounds/Helios/POI/NitroDrome/Fireworks/NitroDrome_Fireworks_Burst_3D_01
    /Game/Sounds/Helios/POI/NitroDrome/Fireworks/NitroDrome_Fireworks_Burst_3D_02
    /Game/Sounds/Helios/POI/NitroDrome/Fireworks/NitroDrome_Fireworks_Burst_3D_03
    /Game/Sounds/Helios/POI/NitroDrome/Fireworks/NitroDrome_Fireworks_Burst_3D_04
    /Game/Sounds/Helios/POI/NitroDrome/Fireworks/NitroDrome_Fireworks_Burst_3D_05
    /Game/Sounds/Helios/POI/NitroDrome/Fireworks/NitroDrome_Fireworks_Burst_3D_06
    /Game/Sounds/Helios/POI/NitroDrome/Fireworks/NitroDrome_Fireworks_Burst_3D_07
    /Game/Sounds/Helios/POI/NitroDrome/Fireworks/NitroDrome_Fireworks_Burst_3D_08
    /Game/Sounds/Helios/POI/NitroDrome/Fireworks/NitroDrome_Fireworks_Burst_3D_09
    /Game/Sounds/Helios/POI/NitroDrome/Fireworks/NitroDrome_Fireworks_Burst_3D_10
    /Game/Sounds/Helios/POI/NitroDrome/Fireworks/NitroDrome_Fireworks_Burst_3D_11
   ]
  }
  nodes [
   n1 = VariableDeferredAccessor.Bool(Variable=InitVariable.Bool().Variable)
   n2 = TriggerCompare.Clamp:Bool(Compare=^NewEvent, A=@n1.Value) {also {defaults {B="<MetaSoundBuilderResult.SUCCEEDED: 0>"}}}
   TriggerAnyTriggerAny2_0 = "TriggerAny.Trigger Any (2)"(In 0=^UE.Source.OnPlay, In 1=@n2.False)
   n3 = VariableAccessor.Trigger(Variable=VariableMutator.Trigger(Variable=InitVariable.Trigger().Variable, Value=@TriggerAnyTriggerAny2_0.Out).Variable)
   n4 = VariableAccessor.Trigger(Variable=@n3.Variable)
   n5 = VariableAccessor.Trigger(Variable=VariableMutator.Trigger(Variable=InitVariable.Trigger().Variable, Value=@n2.True).Variable)
   TriggerDelay_1 = UE.Trigger Delay(In=@n5.Value) {also {defaults {Delay Time=0.01}}}
   TriggerAnyTriggerAny2_1 = "TriggerAny.Trigger Any (2)"(In 1=@TriggerDelay_1.Out, In 0=@n3.Value)
   TriggerDelay_2 = UE.Trigger Delay(Reset=UE.Reroute(Value=UE.Reroute(Value=@n5.Value).Value).Value, In=@TriggerAnyTriggerAny2_1.Out) {also {defaults {Delay Time=2.9}}}
   n6 = VariableAccessor.Trigger(Variable=VariableMutator.Trigger(Variable=InitVariable.Trigger().Variable, Value=@TriggerDelay_2.Out).Variable)
   n7 = VariableAccessor.Trigger(Variable=@n6.Variable)
   TriggerRouteTriggerRouteBool2_0 = "TriggerRoute.Trigger Route (Bool, 2):Bool"(Set 1=@n3.Value, Set 0=VariableAccessor.Trigger(Variable=@n7.Variable).Value) {also {defaults {Value 1="<MetaSoundBuilderResult.SUCCEEDED: 0>"}}}
   n8 = VariableAccessor.Bool(Variable=VariableMutator.Bool(Value=@TriggerRouteTriggerRouteBool2_0.Value, Variable=@n1.Variable).Variable)
   n9 = /FortMetasound/Sources/MSS_PlayRandomSoundLooping_Stereo(Sounds=^LoopingOnshotSource, UE.Source.OnPlay=UE.Reroute(Value=@n4.Value).Value, Loop=@n8.Value) {also {defaults {Attack Time=0, Max Volume=0.8, Min Volume=0.5}}}
   TriggerDelay_0 = UE.Trigger Delay(In=@n4.Value) {also {defaults {Delay Time=3}}}
   n10 = VariableAccessor.Bool(Variable=@n8.Variable)
   n11 = VariableAccessor.Bool(Variable=@n10.Variable)
   n12 = &MSS_PlayRandomSoundLooping_Stereo(
    Sounds=^LoopingOnshotSource
    UE.Source.OnPlay=TriggerCompare.Clamp:Bool(Compare=@TriggerDelay_0.Out, A=@n11.Value) {also {defaults {B="<MetaSoundBuilderResult.SUCCEEDED: 0>"}}}.True
    Loop=@n10.Value
   ) {also {defaults {Attack Time=0, Max Volume=0.8, Min Volume=0.5}}}
   n13 = VariableAccessor.Trigger(Variable=@n4.Variable)
   n14 = &MSS_PlayRandomSound_Stereo(Sounds=^Source, UE.Source.OnPlay=@n13.Value) {also {defaults {MaxVolume=1.05, MinVolume=0.95}}}
   n15 = &MSS_PlayRandomSound_Stereo(Sounds=^OnFinished, UE.Source.OnPlay=@n7.Value)
   n16 = VariableAccessor.Trigger(Variable=@n13.Variable)
   ADSREnvelopeADSREnvelope_0 = ADSR Envelope.ADSR Envelope:Float(Trigger Release=@n6.Value, Trigger Attack=VariableAccessor.Trigger(Variable=@n16.Variable).Value) {also {defaults {Release Time=2, Sustain Level=1}}}
   n17 = UE.Reroute(Value=@ADSREnvelopeADSREnvelope_0.Out Envelope)
   AudioMixerAudioMixerStereo4_0 = "AudioMixer.Audio Mixer (Stereo, 4)"(
    In 3 L=@n9."UE.OutputFormat.Stereo.Audio:0"
    In 3 R=@n9."UE.OutputFormat.Stereo.Audio:1"
    In 2 L=@n12."UE.OutputFormat.Stereo.Audio:0"
    In 2 R=@n12."UE.OutputFormat.Stereo.Audio:1"
    In 1 L=@n14."UE.OutputFormat.Stereo.Audio:0"
    In 1 R=@n14."UE.OutputFormat.Stereo.Audio:1"
    In 0 L=@n15."UE.OutputFormat.Stereo.Audio:0"
    In 0 R=@n15."UE.OutputFormat.Stereo.Audio:1"
    Gain 2=@n17.Value
    Gain 3=@n17.Value
   )
   n18 = InitVariable.Trigger()
   n19 = InitVariable.Trigger()
   TriggerAccumulatorTriggerAccumulate4_0 = "TriggerAccumulator.Trigger Accumulate (4)"(
    In 1=@ADSREnvelopeADSREnvelope_0.On Done
    In 3=VariableAccessor.Trigger(Variable=VariableMutator.Trigger(Variable=@n18.Variable, Value=@n15.UE.Source.OneShot.OnFinished).Variable).Value
    In 2=VariableAccessor.Trigger(Variable=VariableMutator.Trigger(Variable=@n19.Variable, Value=@n14.UE.Source.OneShot.OnFinished).Variable).Value
    In 0=@n16.Value
   ) {also {defaults {Auto Reset="<MetaSoundBuilderResult.SUCCEEDED: 0>"}}}
   TriggerCompareClamp_0 = TriggerCompare.Clamp:Bool(Compare=@TriggerAccumulatorTriggerAccumulate4_0.Out, A=VariableAccessor.Bool(Variable=@n11.Variable).Value) {also {defaults {B="<MetaSoundBuilderResult.SUCCEEDED: 0>"}}}
  ]
  links [
   {dst="@UE.OutputFormat.Stereo.Audio:0", src=AudioMixerAudioMixerStereo4_0.Out L}
   {dst="@UE.OutputFormat.Stereo.Audio:1", src=AudioMixerAudioMixerStereo4_0.Out R}
   {dst="@UE.Source.OneShot.OnFinished", src=TriggerCompareClamp_0.False}
  ]
 }]
properties {
 assetImportData=&self:AssetImportData
 attenuationSettings=&/NitroDrome_Sparks_Att
 bOverrideConcurrency=true
 maxDistance=16000
 numChannels=2
 outputFormat=Stereo
 sampleRate=48000
 soundClassObject=&Fort_Ambient_Athena
 volume=0.952
 concurrencyOverrides {maxCount=3}
}
