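# Stereo one-shot MetaSound Source preset for distant firework sparks at the Helios NitroDrome POI, derived from MSS_Helios_Nitrodrome_Sparks.
# The parent supplies the graph; this preset lists wave arrays of 3D bursts (Source), sparks (LoopingOnshotSource) and spark endings (OnFinished).
# It plays in the Athena ambient sound class with NitroDrome sparks attenuation and a concurrency override capping simultaneous instances.
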MetaSoundSource /Game/Sounds/Helios/POI/NitroDrome/MSS_Helios_Nitrodrome_Sparks_Distant
preset_parent=&/MSS_Helios_Nitrodrome_Sparks
inputs [
 LoopingOnshotSource {type=WaveAsset:Array}
 NewEvent {type=Trigger}
 OnFinished {type=WaveAsset:Array}
 Source {type=WaveAsset:Array}
 UE.Source.OnPlay {type=Trigger}
]
interfaces [UE.OutputFormat.Stereo:1.0, UE.Source.OneShot:1.0, UE.Source:1.1]
outputs [
 "UE.OutputFormat.Stereo.Audio:0" {type=Audio}
 "UE.OutputFormat.Stereo.Audio:1" {type=Audio}
 UE.Source.OneShot.OnFinished {type=Trigger}
]
pages [Default {input_defaults {
   NewEvent=null
   UE.Source.OnPlay=null
   LoopingOnshotSource [
    /Game/Sounds/Helios/POI/NitroDrome/Fireworks/NitroDrome_Fireworks_Sparks_01
    /Game/Sounds/Helios/POI/NitroDrome/Fireworks/NitroDrome_Fireworks_Sparks_02
    /Game/Sounds/Helios/POI/NitroDrome/Fireworks/NitroDrome_Fireworks_Sparks_03
    /Game/Sounds/Helios/POI/NitroDrome/Fireworks/NitroDrome_Fireworks_Sparks_04
   ]
   OnFinished [
    /Game/Sounds/Helios/POI/NitroDrome/Fireworks/NitroDrome_Fireworks_Sparks_End_04
    /Game/Sounds/Helios/POI/NitroDrome/Fireworks/NitroDrome_Fireworks_Sparks_End_03
    /Game/Sounds/Helios/POI/NitroDrome/Fireworks/NitroDrome_Fireworks_Sparks_End_02
    /Game/Sounds/Helios/POI/NitroDrome/Fireworks/NitroDrome_Fireworks_Sparks_End_01
   ]
   Source [
    /Game/Sounds/Helios/POI/NitroDrome/Fireworks/NitroDrome_Fireworks_Burst_3D_01
    /Game/Sounds/Helios/POI/NitroDrome/Fireworks/NitroDrome_Fireworks_Burst_3D_02
    /Game/Sounds/Helios/POI/NitroDrome/Fireworks/NitroDrome_Fireworks_Burst_3D_03
    /Game/Sounds/Helios/POI/NitroDrome/Fireworks/NitroDrome_Fireworks_Burst_3D_04
    /Game/Sounds/Helios/POI/NitroDrome/Fireworks/NitroDrome_Fireworks_Burst_3D_05
    /Game/Sounds/Helios/POI/NitroDrome/Fireworks/NitroDrome_Fireworks_Burst_3D_06
    /Game/Sounds/Helios/POI/NitroDrome/Fireworks/NitroDrome_Fireworks_Burst_3D_07
    /Game/Sounds/Helios/POI/NitroDrome/Fireworks/NitroDrome_Fireworks_Burst_3D_08
    /Game/Sounds/Helios/POI/NitroDrome/Fireworks/NitroDrome_Fireworks_Burst_3D_09
    /Game/Sounds/Helios/POI/NitroDrome/Fireworks/NitroDrome_Fireworks_Burst_3D_10
    /Game/Sounds/Helios/POI/NitroDrome/Fireworks/NitroDrome_Fireworks_Burst_3D_11
   ]
  }}]
properties {
 assetImportData=&self:AssetImportData
 attenuationSettings=&/NitroDrome_Sparks_Att
 bOverrideConcurrency=true
 maxDistance=16000
 numChannels=2
 outputFormat=Stereo
 soundClassObject=&Fort_Ambient_Athena
 volume=1.128
 concurrencyOverrides {maxCount=3}
}
