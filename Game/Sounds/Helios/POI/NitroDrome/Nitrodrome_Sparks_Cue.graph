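# Sound cue for spark ambience at the Nitrodrome POI, assigned to the Athena ambient sound class.
# A single wave player feeds an attenuation node using the Nitrodrome sparks attenuation asset, which feeds a root mixer whose second input is empty.
# A concurrency override caps simultaneous instances at a small count, and the cue's overall volume is trimmed slightly.
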SoundCue /Game/Sounds/Helios/POI/NitroDrome/Nitrodrome_Sparks_Cue
root=Mixer_0
nodes [
 {id=Attenuation_0, children [WavePlayer_0], properties {attenuationSettings=&/NitroDrome_Sparks_Att}}
 {id=Mixer_0, children [Attenuation_0, null], properties {inputVolume [1,1]}}
 {id=WavePlayer_0, properties {soundWaveAssetPtr=&/MSS_Helios_Nitrodrome_Sparks}}
]
properties {
 bOverrideConcurrency=true
 maxDistance=16000
 soundClassObject=&Fort_Ambient_Athena
 volumeMultiplier=0.902
 concurrencyOverrides {maxCount=3}
}
